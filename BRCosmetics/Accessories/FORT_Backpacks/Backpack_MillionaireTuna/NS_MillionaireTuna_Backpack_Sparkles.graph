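# Niagara system for the MillionaireTuna backpack cosmetic that emits short-lived orange-gold sparkles from a single emitter with sprite and mesh renderers.
# Each spawned particle takes its position from a small hard-coded array of points via SelectPositionFromArray.
# Curves grow and shrink sprite size and fade alpha in and out over lifetime, with a slight camera offset.
# A second ScaleColor dims RGB at night through a time-of-day function.
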NiagaraSystem /BRCosmetics/Accessories/FORT_Backpacks/Backpack_MillionaireTuna/NS_MillionaireTuna_Backpack_Sparkles
dependencies {
 usedDataInterfaces [&NiagaraDataInterfaceArrayPosition, &NiagaraDataInterfaceCurve, &NiagaraDataInterfaceColorCurve]
 usedDynamicInputs []
 usedModules [&SystemState]
}
emitters {Sparkles {
  emitterUpdateScript [
   &EmitterState()
   &SpawnRate(Use Spawn Probability=false)
  ]
  particleSpawnScript [
   &InitializeParticle(
    Color=<rgba 1 0.4684 0 1>
    Lifetime Max=0.7
    Lifetime Min=0.35
    Lifetime Mode=::NewEnumerator1
    Mesh Scale Mode=::NewEnumerator3
    Mesh Uniform Scale=2
    Position Offset=<-12 -20 5>
    Sprite Size Mode=::NewEnumerator1
    Uniform Sprite Size Max=31.5
    Uniform Sprite Size Min=10.5
    UsePositionOffset=false
    Write Material Random=true
   )
   /SetVariables(Particles.Position Array=DataInterface() {propertyValues {GpuSyncMode=SyncCpuToGpu, MaxElements=0, PositionData [<10 -18 7>, <-10 -19 -14>, <-10 -18 8>, <-10 -5 8>]}}) {isSetParameters=true}
   /SetVariables(Particles.Position=/Niagara/DynamicInputs/Arrays/SelectPositionFromArray(
     Array Sampling Mode:ENiagaraArraySamplingMode=::NewEnumerator0
     Direct Array Index:NiagaraInt32=?
     Interpolation:NiagaraFloat=?
     Direct Array Mode:ENiagara_WrapClamp=::NewEnumerator0
     Evaluation Type:ENiagaraRandomnessEvaluation=::NewEnumerator0
     Normalized Interpolation:NiagaraBool=true
     Recalculate Random Each Loop:NiagaraBool=false
     Position Array:NiagaraDataInterfaceArrayPosition=$Particles.Position Array
     Randomness Mode:ENiagaraRandomnessMode=::NewEnumerator0
     Fixed Random Seed:NiagaraBool=false
     Random Seed:NiagaraInt32=0
    )) {isSetParameters=true}
  ]
  particleUpdateScript [
   &ParticleState()
   &ScaleSpriteSize(Uniform Curve Sprite Scale=DataInterface() {propertyValues {Curve {keys [{tangentMode=RCTM_User, time=0, value=0}, <at 0.5 0.95>, {tangentMode=RCTM_User, time=1, value=0}]}}})
   &CameraOffset(Camera Offset Amount=10)
   &ScaleColor(
    Linear Color Curve=DataInterface() {propertyValues {
      AlphaCurve {keys [<at 0.0005 0>,<at 0.4223 1>,<at 1 0>]}
      BlueCurve {keys [<at 0 1>,<at 1 1>]}
      GreenCurve {keys [<at 0 1>,<at 1 1>]}
      RedCurve {keys [<at 0 1>,<at 1 1>]}
     }}
    Scale Mode=::NewEnumerator2
   )
   &ScaleColor(
    Scale RGB=&Multiply_VectorByFloat(Vector=<1 1 1>, Float=&FN_Select_Float_from_Time_Of_Day(Morning=1, Day=1, Evening=1, Night=0.6))
    ScaleA=true
    ScaleRGB=true
   ) {module=ScaleColor001}
  ]
  rendererClasses [&NiagaraSpriteRendererProperties, &NiagaraMeshRendererProperties]
  renderers [
   {rendererClass=&NiagaraSpriteRendererProperties}
   {rendererClass=&NiagaraMeshRendererProperties, rendererIndex=1}
  ]
 }}
summary {
 systemName=NS_MillionaireTuna_Backpack_Sparkles
 emitters [{emitterName=Sparkles, rendererClasses [&NiagaraSpriteRendererProperties, &NiagaraMeshRendererProperties]}]
 userVariables []
}
